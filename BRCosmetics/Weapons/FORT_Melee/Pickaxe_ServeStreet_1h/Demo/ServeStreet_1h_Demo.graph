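# Default Lit material for the one-handed ServeStreet pickaxe demo, flagged for skeletal, instanced static and Niagara particle meshes.
# It holds samples of the pickaxe's diffuse and normal textures beside generic parameters for Metal, Roughness, Specular, Normal and Emissive.
# A separate chain multiplies SpecularMasks, Diffuse and CustomizationMask parameters, clamps the result and tints it with Color.
# The graph shows no links from these nodes to material outputs.
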Material /BRCosmetics/Weapons/FORT_Melee/Pickaxe_ServeStreet_1h/Demo/ServeStreet_1h_Demo
graph {
 properties {
  bAutomaticallySetUsageInEditor=false
  bUsedWithInstancedStaticMeshes=true
  bUsedWithMeshDeformer=true
  bUsedWithNiagaraMeshParticles=true
  bUsedWithSkeletalMesh=true
  pixelDepthOffsetMode=PDOM_Legacy
  usedShadingModels=Default Lit
 }
 nodes [
  Constant3Vector() {constant=<rgba 0.125 0.125 0.125 0>}
  n1 = Multiply(
   A=Multiply(
    A=TextureSampleParameter2D() {parameterName=SpecularMasks, texture=/Game/Global/Textures/Default/Blanks/T_Blank_LegalWhite_sRGBColor}.RGB
    B=TextureSampleParameter2D() {parameterName=Diffuse, texture=&T_Blank_LegalWhite_sRGBColor}.RGB
   )
   B=TextureSampleParameter2D() {parameterName="CustomizationMask (_CM)", texture=&T_Blank_LegalWhite_sRGBColor}.RGB
  )
  Multiply(A=VectorParameter() {parameterName=Color, defaultValue=<rgba 0.3333 0.3333 0.3333 1>}.RGB, B=Max(A=Clamp(None=@n1)))
  ScalarParameter() {defaultValue=0.1, parameterName=Metal}
  ScalarParameter() {defaultValue=0.3, parameterName=Roughness}
  ScalarParameter() {parameterName=Specular}
  TextureSampleParameter2D() {parameterName=Normal, samplerType=SAMPLERTYPE_Normal, texture=/Game/Athena/Prototype/HardSurface/Textures/HS_FlatNormal}
  TextureSample() {texture=&/T_ServeStreet_1h_Demo_D}
  TextureSample() {samplerType=SAMPLERTYPE_Normal, texture=&/T_Pickaxe_ServeStreet_N}
  VectorParameter() {parameterName=Emissive}
 ]
}
